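# One-shot mono MetaSound Source that plays the alarm sound for the weapon bunker doors.
# The play trigger starts a single Wave Player on the Alarm_v2 wave, which feeds the mono output and signals finished when the wave ends.
# It uses a bunker alarm attenuation asset and the in-game SFX sound class, and keeps playing when silent.
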MetaSoundSource /MidMatchWeaponBunkerGameplay/Items/WeaponBunkerDoor/Audio/MSS_Bunker_Doors_Alarm
inputs [UE.Source.OnPlay {type=Trigger}]
interfaces [UE.OutputFormat.Mono:1.0, UE.Source.OneShot:1.0, UE.Source:1.1]
outputs [
 "UE.OutputFormat.Mono.Audio:0" {type=Audio}
 UE.Source.OneShot.OnFinished {type=Trigger}
]
pages [Default {
  input_defaults {UE.Source.OnPlay=null}
  nodes [WavePlayer_0 = UE.Wave Player:Mono(Play=^UE.Source.OnPlay) {also {defaults {Wave Asset=/MidMatchWeaponBunkerGameplay/Items/WeaponBunkerDoor/Audio/Waves/Alarm_v2}}}]
  links [{dst="@UE.Source.OneShot.OnFinished", src=WavePlayer_0.On Finished}, {dst="@UE.OutputFormat.Mono.Audio:0", src=WavePlayer_0.Out Mono}]
 }]
properties {
 assetImportData=&self:AssetImportData
 attenuationSettings=/MidMatchWeaponBunkerGameplay/Items/WeaponBunkerDoor/MMOBunkers_Ch6S4/Sound/Waves/ATT_TidalCrane_Bunker_Alarm
 bDebug=true
 maxDistance=34200
 sampleRate=48000
 soundClassObject=&Fort_SFX_InGame
 virtualizationMode=PlayWhenSilent
 volume=3
 busSends [{audioBus=null, sendLevel=1}]
}
