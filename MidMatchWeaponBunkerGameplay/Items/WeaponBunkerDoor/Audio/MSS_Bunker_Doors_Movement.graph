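# Mono one-shot MetaSound for the weapon bunker door's opening movement.
# Open.Start plays a start wave and a looping movement layer; Open.Finish plays an end wave and stops the loop after an envelope release.
# A master envelope releases when a wave-finished index check passes or a Duration-based fallback delay elapses, then ends the source.
# The bIsLeftDoor input is unused, and the source feeds an ambient sidechain bus.
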MetaSoundSource /MidMatchWeaponBunkerGameplay/Items/WeaponBunkerDoor/Audio/MSS_Bunker_Doors_Movement
inputs [
 bIsLeftDoor {type=Bool}
 Duration {type=Float}
 Open.Finish {type=Trigger}
 Open.Start {type=Trigger}
 UE.Source.OnPlay {type=Trigger}
 Wavs {type=WaveAsset:Array}
]
interfaces [UE.OutputFormat.Mono:1.0, UE.Source.OneShot:1.0, UE.Source:1.1]
outputs [
 "UE.OutputFormat.Mono.Audio:0" {type=Audio}
 UE.Source.OneShot.OnFinished {type=Trigger}
]
pages [Default {
  input_defaults {Duration=15, Open.Finish=null, Open.Start=null, UE.Source.OnPlay=null, bIsLeftDoor=null, Wavs []}
  nodes [
   InitVariable.WaveAsset:Array()
   TriggerRouteTriggerRouteInt322_0 = "TriggerRoute.Trigger Route (Int32, 2):Int32"(Set 1=^Open.Finish, Set 0=^Open.Start) {also {defaults {Value 1=2}}}
   n1 = UE.Reroute(Value=@TriggerRouteTriggerRouteInt322_0.Value)
   n2 = UE.Reroute(Value=@TriggerRouteTriggerRouteInt322_0.Value)
   WavePlayer_0 = UE.Wave Player:Mono(
    Wave Asset=UE.Reroute(Value=Array.Get:WaveAsset:Array(Array=^Wavs, Index=@n2.Value, Trigger=@TriggerRouteTriggerRouteInt322_0.On Set).Element).Value
    Play=@TriggerRouteTriggerRouteInt322_0.On Set
   )
   TriggerDelay_0 = UE.Trigger Delay(
    In=^Open.Start
    Delay Time=UE.Add:Time(PrimaryOperand=UE.ConversionFloatToTime(In=^Duration).Out) {also {defaults {AdditionalOperands=1}}}.Out
   )
   TriggerAnyTriggerAny2_0 = "TriggerAny.Trigger Any (2)"(
    In 1=TriggerCompare.Clamp:Int32(A=UE.Reroute(Value=@n1.Value).Value, Compare=@WavePlayer_0.On Finished) {also {defaults {B=0, Type=1}}}.True
    In 0=@TriggerDelay_0.Out
   )
   ADSREnvelopeADSREnvelope_0 = ADSR Envelope.ADSR Envelope:Audio(Trigger Attack=^UE.Source.OnPlay, Trigger Release=@TriggerAnyTriggerAny2_0.Out) {also {defaults {Release Time=0.5, Sustain Level=1}}}
   DecibelstoLinearGain_0 = UE.Decibels to Linear Gain:Float() {also {defaults {Decibels=4}}}
   ADSREnvelopeADSREnvelope_1 = ADSR Envelope.ADSR Envelope:Float(Trigger Release=^Open.Finish, Trigger Attack=^Open.Start) {also {defaults {Attack Time=0.5, Release Time=0.5, Sustain Level=1}}}
   DecibelstoLinearGain_1 = UE.Decibels to Linear Gain:Float(Decibels=UE.Multiply:Float(PrimaryOperand=@ADSREnvelopeADSREnvelope_1.Out Envelope) {also {defaults {AdditionalOperands=2}}}.Out)
   WavePlayer_1 = UE.Wave Player:Mono(
    Play=^Open.Start
    Stop=@ADSREnvelopeADSREnvelope_1.On Done
    Wave Asset=Array.Get:WaveAsset:Array(Trigger=^Open.Start, Array=^Wavs) {also {defaults {Index=1}}}.Element
   ) {also {defaults {Loop="<MetaSoundBuilderResult.SUCCEEDED: 0>"}}}
   AudioMixerAudioMixerMono2_0 = "AudioMixer.Audio Mixer (Mono, 2)"(
    Gain 0=@DecibelstoLinearGain_0.Linear Gain
    Gain 1=@DecibelstoLinearGain_1.Linear Gain
    In 0=@WavePlayer_0.Out Mono
    In 1=@WavePlayer_1.Out Mono
   )
   Multiply_0 = UE.Multiply:Audio(AdditionalOperands=@ADSREnvelopeADSREnvelope_0.Out Envelope, PrimaryOperand=@AudioMixerAudioMixerMono2_0.Out)
  ]
  links [{dst="@UE.Source.OneShot.OnFinished", src=ADSREnvelopeADSREnvelope_0.On Done}, {dst="@UE.OutputFormat.Mono.Audio:0", src=Multiply_0.Out}]
 }]
properties {
 assetImportData=&self:AssetImportData
 attenuationSettings=&/ATT_Bunker_Opening_Doors
 bDebug=true
 maxDistance=26575.002
 sampleRate=48000
 soundClassObject=&Fort_SFX_InGame
 virtualizationMode=PlayWhenSilent
 volume=2.2
 busSends [{audioBus=&AB_Sidechain_FortAmbient_Vol, sendLevel=0.5}]
}
