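; Weapon bunker door actor for a mid-match objective that primes, then opens after a hotfixable delay when StartOpeningSequence is called.
; The server drives replicated state from Closed through Priming, Opening and Opened, updating map icon and marking.
; OnRep handlers run persistent klaxon, smoke, beacon and audio effects plus the timeline door motion, while an RPC plays burst effects.
; An optional analytics volume records players entering during a timed window from priming.
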(event EventBeginPlay
  (bind _pointlightklaxon01 (Variables|Default|GetPointLightKlaxon01))
  ;[0.06801 0.5677 0.1422] BeginPlay: At the start of the match, the bunker door state is closed.
  (|CallParentFunction)
  (Networking|SwitchHasAuthority
    (:Authority
      (Class|FortAsyncActionCurrentPlaylistReady|PlaylistDataReadyAsync
        (:OnReady
          (CallFunction|SetWeaponBunkerDoorState "NewEnumerator6")
          (CallFunction|CreateMapIcon)
          (if (ScalableFloat|GetValueasBool (Variables|Hotfix|GetHotfix_AnalyticsEnabled))
            (bind _returnvalue (Game|SpawnActorfromClass "/MidMatchWeaponBunkerGameplay/Analytics/B_WeaponBunkerAnalyticsVolumeActor.B_WeaponBunkerAnalyticsVolumeActor_C" (Math|Conversions|ToTransform(Vector) (Transformation|GetActorLocation)) "AlwaysSpawn" "MultiplyWithRoot" 0 (Variables|General|GetBunkerName) 0))
            (bind _output_get (Variables|Transient|SetAnalyticsVolumeActor _returnvalue))
            (Utilities|IsValid _output_get
              (:Is%20Valid
                (Class|BWeaponBunkerAnalyticsVolumeActor|SetEnableAnalyticsVolume _output_get false)))))))
    (:Remote
      (bind _returnvalue_2 (Rendering|Material|CreateDynamicMaterialInstance "/MidMatchWeaponBunkerGameplay/Items/WeaponBunkerDoor/M_LF_WeaponBunkerKlaxon.M_LF_WeaponBunkerKlaxon" "None" "None" :!class /Script/Engine.KismetMaterialLibrary))
      (bind _output_get_2 (Variables|Materials|SetKlaxonLightFunctionMID _returnvalue_2))
      (Rendering|Components|Light|SetLightFunctionMaterial _pointlightklaxon01 _output_get_2)
      (Variables|Materials|SetKlaxonLightIntensity (Variables|/Script/Engine.LightComponentBase|GetIntensity _pointlightklaxon01)))))

(event Custom|StartOpeningSequence
  ;[0.2083 0.04968 0.1819] StartOpeningSequence: Switches the bunker door to the priming state and after the desired amount of delay, switches it to the Opening state to start opening the door.
  ; Called from the objective's game state component.
  (Networking|SwitchHasAuthority
    (:Authority
      (Utilities|FlowControl|DoOnce false
        (:Completed
          (CallFunction|SetWeaponBunkerDoorState "NewEnumerator1")
          (Utilities|FlowControl|Delay (CallFunction|GetPrimingDuration))
          (CallFunction|SetWeaponBunkerDoorState "NewEnumerator2"))))))

(event Custom|OnRep_WeaponBunkerDoorState
  ;[0.6805 0.3014 1] OnRep_WeaponBunkerDoorOpeningState: Handle the various states of the weapon bunker door actor.
  ; Called on both server and client.
  ;;[0.214 0.5449 1] Log the new state on the client and server.
  (Fortnite|FortShippingLog self (Utilities|String|Append "OnRep_WeaponBunkerDoorState: State changed to : " (Utilities|String|EnumtoString (Variables|Transient|GetWeaponBunkerDoorState))))
  (switch Utilities|FlowControl|Switch|SwitchonWeaponBunkerDoorOpeningStateEnum (Variables|Transient|GetWeaponBunkerDoorState)
    (:NewEnumerator6
      ;;[0.1557 0.8701 1] Handle Closed state. Default state of the actor before the objective is activated.
      (Networking|SwitchHasAuthority
        (:Authority
          (Variables|StateTrackers|SetbMarkTraceResponseEnabled false) ; Do NOT allow players to mark this actor in the CLOSED state.
        )))
    (:NewEnumerator1
      ;;[0.155 0.8698 0.5069] Handle the Priming state.  Enable the priming fan-fare fx and allow players to mark this actor.
      (Networking|SwitchHasAuthority
        (:Authority
          (Variables|StateTrackers|SetbPrimingEnabled true) ; On the client, enable the priming fan-fare fx.
          (Variables|StateTrackers|SetbMarkTraceResponseEnabled true) ; Allow players to mark this actor.
          (CallFunction|SetMapIconOpening)
          (bind _analyticsvolumeactor (Variables|Transient|GetAnalyticsVolumeActor)
            (:then
              (Class|BWeaponBunkerAnalyticsVolumeActor|SetEnableAnalyticsVolume _analyticsvolumeactor true)
              ;;;[0.01757 0.1406 0.1392] Setup a timer, for the duration of which analytics information is collected.
              ;;; Note: This is only run if the analytics volume actor is valid, i.e, analytics is enabled.
              (bind _returnvalue (Utilities|Time|SetTimerbyEvent (EventDispatchers|CreateEvent :!event HandleAnalyticsCollectionDurationComplete) (Math|Float|Max(Float) (ScalableFloat|GetValueatLevel (Variables|Hotfix|GetHotfix_AnalyticsCollectionDuration)) 1.0)))
              (Variables|Transient|SetAnalyticsCollectionDurationTimerHandle _returnvalue)
              (Collision|SetCollisionEnabled (Variables|Default|GetAnalyticsPlayerCollisionVolume) "QueryOnly"))
            (:else)))
        (:Remote
          (CallFunction|EnableKlaxxonPulsing 1.5)
          (CallFunction|FadeKlaxonLightIn)
          (Components|Activation|Activate (Variables|Default|GetNS_WeaponBunkerDoor_Beacon) false))))
    (:NewEnumerator2
      ;;[0.4271 0.402 0.000162] Handle opening the vault door.
      (Networking|SwitchHasAuthority
        (:Authority
          (Components|Timeline|SetPlayRate (Variables|BGA_WeaponBunkerDoor|GetTimeline_DoorOpen) (Math|Float|SafeDivide 1.0 (Math|Float|Max(Float) (ScalableFloat|GetValueatLevel (Variables|Hotfix|GetHotfix_DoorOpeningDuration)) 1.0)))
          ;;;[0.02991 0.1225 0.125] Replicated timeline to update the rotation of the door mesh component on both the server and client smoothy.
          (bind (_direction _rotationalpha) (Timeline|Timeline_DoorOpen Play)
            (:Update
              (bind _staticmeshdoor (Variables|Default|GetStaticMeshDoor)
                (:then
                  (Transformation|SetRelativeTransform _staticmeshdoor (Math|Transform|MakeTransform (Math|Vector|MakeVector 0.0 (Variables|Setting|GetDoorRelativeLocationY) (Math|Float|Lerp (Variables|Setting|GetDoorOpenRelativeLocationZStart) (Variables|Setting|GetDoorOpenRelativeLocationZEnd) _rotationalpha)) (Math|Rotator|MakeRotator (Math|Float|Lerp (Variables|Setting|GetDoorOpenRollAngleStart) (Variables|Setting|GetDoorOpenRollAngleEnd) _rotationalpha))) false false)
                  (Networking|FlushNetDormancy))
                (:else)))
            (:Finished
              (Networking|SwitchHasAuthority
                (:Authority
                  (CallFunction|SetWeaponBunkerDoorState "NewEnumerator5") ; Update the door state to Opened after the timeline is complete.
                )))))
        (:Remote
          (CallFunction|EnableKlaxxonPulsing 1.5)
          (Components|Activation|Activate (Variables|Default|GetNiagara_Smoke) false)
          (Niagara|SetNiagaraVariable(LinearColor) (Variables|Default|GetNS_WeaponBunkerDoor_Beacon) "Color" (Variables|VFX|GetBeacon%20Color%20Opening)))))
    (:NewEnumerator5
      ;;[0.006342 0.2109 0.2292] Handle the opened state.
      (Transformation|SetRelativeTransform (Variables|Default|GetStaticMeshDoor) (Math|Transform|MakeTransform (Math|Vector|MakeVector 0.0 (Variables|Setting|GetDoorRelativeLocationY) (Variables|Setting|GetDoorOpenRelativeLocationZEnd)) (Math|Rotator|MakeRotator (Variables|Setting|GetDoorOpenRollAngleEnd))) false false) ; Set the final transform of the static mesh door, \nuseful on the client if they missed the opening state.
      (CallFunction|RefreshMarkerDisplayInfo)
      (Networking|SwitchHasAuthority
        (:Authority
          (Variables|StateTrackers|SetbPrimingEnabled false) ; Disable the priming fan-fare
          (CallFunction|SetMapIconOpened))
        (:Remote
          (CallFunction|EnableKlaxxonPulsing 1.5)
          (Components|Activation|Deactivate (Variables|Default|GetNiagara_Smoke))
          (CallFunction|ChangeLightColor "(R=1.000000,G=0.526355,B=0.000000,A=1.000000)")
          (Rendering|Material|SetVectorParameterValue (Class|BTechArtGlobalLibrary|GetDynamicMaterialInstance (Variables|Default|GetFlare01)) "FlareColor" (Variables|Materials|GetFlareColorOpened) :!class /Script/Engine.MaterialInstanceDynamic)
          (Rendering|Material|SetVectorParameterValue (Class|BTechArtGlobalLibrary|GetDynamicMaterialInstance (Variables|Default|GetStaticMeshKlaxon01)) "SFX_Emissive_Tint(Expand)" (Variables|Materials|GetKlaxonOpenedEmissive) :!class /Script/Engine.MaterialInstanceDynamic)
          (Components|Activation|Activate (Variables|Default|GetAudio_Amb_Alarm) false))))))

(event Custom|ClientWeaponBunkerDoorStateChanged
  (params (NewState :type WeaponBunkerDoorOpeningStateEnum%20Enum :default "NewEnumerator0"))
  (bind _audio_priming (Variables|Default|GetAudio_Priming))
  (bind _returnvalue (Transformation|GetWorldLocation _audio_priming))
  ;[0.332 1 0.8799] ClientWeaponBunkerDoorStateChanged: This RPC event is used to handle burst effects on the client and will be fired right as the state update happens.
  (switch Utilities|FlowControl|Switch|SwitchonWeaponBunkerDoorOpeningStateEnum NewState
    (:NewEnumerator1
      (Audio|SpawnSoundatLocation (Variables|Sound|GetSound_PrimingTell) (Transformation|GetWorldLocation (Variables|Default|GetAudio_Priming)))
      (Utilities|FlowControl|Delay 0.2)
      (Audio|Parameter|SetFloatParameter (Variables|Default|GetAudio_Priming) "Priming Duration" (CallFunction|GetPrimingDuration) :!class /Script/AudioExtensions.AudioParameterControllerInterface)
      (Utilities|FlowControl|Sequence
        (:then_0
          (Utilities|FlowControl|Delay 2.0)
          (Components|Activation|Activate (Variables|Default|GetNiagara_Smoke_Priming) false)
          (Utilities|FlowControl|Delay 3.0)
          (Components|Activation|Deactivate (Variables|Default|GetNS_WeaponBunker_DoorDust_Idle)))
        (:then_1
          (Utilities|FlowControl|Delay 4.0)
          (Audio|Components|Audio|Play (Variables|Default|GetAudio_UndergoundMech) 0.0))))
    (:NewEnumerator2
      (Audio|SpawnSoundatLocation (Variables|Sound|GetSound_DoorOpening) _returnvalue)
      (Audio|SpawnSoundatLocation (Variables|Sound|GetSound_DoorOpeningAlert) _returnvalue)
      (bind _audio_smoke (Variables|Default|GetAudio_Smoke)
        (:then
          (Audio|Components|Audio|FadeOut _audio_smoke 6.0 0.0 "Linear" :!class /Script/Engine.AudioComponent))
        (:else)))))

(event Custom|OnRep_bMarkTraceResponseEnabled
  ;[0.1745 0.409 1] OnRep_bMarkTraceResponseEnabled: Update the response to the trace channel to enable or disable client side player marking.
  (Networking|CanExecuteCosmeticEvents
    (:True
      (Collision|SetCollisionResponsetoChannel (Actor|GetComponentsByClass "/Script/Engine.StaticMeshComponent") "ECC_GameTraceChannel1" (select (Variables|StateTrackers|GetbMarkTraceResponseEnabled) "ECR_Block" "ECR_Ignore")))))

(event Custom|OnRep_bPrimingEnabled
  (bind _audio_priming (Variables|Default|GetAudio_Priming))
  ;[0.1094 0.02209 0.07442] OnRep_bPrimingEnabled: Enable or disable fanfare visuals. Note that BURST effects should be hooked up in event ClientWeaponBunkerDoorStateChanged. This function should only handle looping/consistent states visuals such as material changes and enable/disabling looping audio components.
  (Networking|CanExecuteCosmeticEvents
    (:True
      (if (Variables|StateTrackers|GetbPrimingEnabled)
        (Utilities|FlowControl|Delay 3.5)
        (Components|Activation|Activate _audio_priming false)
        (Audio|Components|Audio|Play _audio_priming 0.0)
        (SoundIndicator|ShowSoundIndicatorforDuration (Variables|Default|GetFortSoundIndicator) (Transformation|GetActorLocation) (Variables|/Script/Engine.AudioComponent|GetSound _audio_priming) 1.0 25.0 0)
        (else
          (Audio|Components|Audio|FadeOut (Variables|Default|GetAudio_Priming) 1.0 0.0 "Linear" :!class /Script/Engine.AudioComponent)
          (SoundIndicator|StopSoundIndicatorLooping (Variables|Default|GetFortSoundIndicator)))))))

(event Custom|EnableKlaxxonPulsing
  (params (PulseDuration :type Float%20%28double-precision%29))
  (bind _staticmeshklaxon01 (Variables|Default|GetStaticMeshKlaxon01))
  (bind _material_instance_dynamic (Class|BTechArtGlobalLibrary|GetDynamicMaterialInstance _staticmeshklaxon01))
  (bind _returnvalue (Math|Float|Max(Float) PulseDuration 0.1))
  (bind _returnvalue_2 (/ 1.0 _returnvalue))
  ;[0.01402 0.08333 0.05784] Enable or disable Klaxxon pulsing.
  (Networking|CanExecuteCosmeticEvents
    (:True
      (Rendering|SetVisibility (Variables|Default|GetPointLightKlaxon01) true false)
      (Rendering|Material|SetVectorParameterValue _material_instance_dynamic "SFX_Emissive_Tint(Expand)" (Variables|Materials|GetKlaxonOpeningEmissive) :!class /Script/Engine.MaterialInstanceDynamic)
      (Rendering|Material|SetScalarParameterValue _material_instance_dynamic "Emissive Time Speed Layer 1" _returnvalue_2 :!class /Script/Engine.MaterialInstanceDynamic)
      (Rendering|Material|SetScalarParameterValue (Variables|Materials|GetKlaxonLightFunctionMID) "Speed" _returnvalue_2 :!class /Script/Engine.MaterialInstanceDynamic)
      (Rendering|Material|SetScalarParameterValue (Class|BTechArtGlobalLibrary|GetDynamicMaterialInstance (Variables|Default|GetFlare01)) "Speed" _returnvalue_2 :!class /Script/Engine.MaterialInstanceDynamic))))

(event Custom|DisableKlaxxonPulsing
  ;[0.01402 0.08333 0.05784] Enable or disable Klaxxon pulsing.
  (Networking|CanExecuteCosmeticEvents
    (:True
      (Rendering|SetVisibility (Variables|Default|GetPointLightKlaxon01) false false))))

(event Custom|FadeKlaxonLightIn
  ;[0.01402 0.08333 0.05784] Enable or disable Klaxxon pulsing.
  (bind (_direction _fade) (Timeline|FadeKlaxonLightTL PlayFromStart)
    (:Update
      (Rendering|Components|Light|SetIntensity (Variables|Default|GetPointLightKlaxon01) (Math|Float|Lerp 0.0 (Variables|Materials|GetKlaxonLightIntensity) _fade)))))

(event Custom|ChangeLightColor
  (params (NewColor :type Linear%20Color%20Structure))
  ;[0.01402 0.08333 0.05784] Enable or disable Klaxxon pulsing.
  (if (not (Components|Timeline|IsPlaying (Variables|BGA_WeaponBunkerDoor|GetChangeLightColorTL)))
    (Variables|Materials|SetTargetLightColor NewColor)
    (Variables|Materials|SetCurrentLightColor (Rendering|Components|Light|GetLightColor (Variables|Default|GetPointLightKlaxon01)))
    (bind (_direction _lerp) (Timeline|ChangeLightColorTL PlayFromStart)
      (:Update
        (Rendering|Components|Light|SetLightColor (Variables|Default|GetPointLightKlaxon01) (Math|Color|Lerp(LinearColor) (Variables|Materials|GetCurrentLightColor) (Variables|Materials|GetTargetLightColor) _lerp) true)))))

(event Custom|FadeKlaxonLightOut
  ;[0.01402 0.08333 0.05784] Enable or disable Klaxxon pulsing.
  (Timeline|FadeKlaxonLightTL ReverseFromEnd))

(event EventEndPlay
  (params (EndPlayReason :type EEndPlayReason%20Enum :default "Destroyed"))
  ;[0.224 0.1449 0.04084] EndPlay: Destroy the map icon.
  (Utilities|FlowControl|Sequence
    (:then_0
      (Networking|SwitchHasAuthority
        (:Authority
          (CallFunction|DestroyMapIcon)
          (bind _analyticsvolumeactor (Variables|Transient|GetAnalyticsVolumeActor)
            (:then
              (Actor|DestroyActor _analyticsvolumeactor))
            (:else)))))
    (:then_1
      (|CallParentFunction EndPlayReason))))

(event OnComponentBeginOverlap(AnalyticsPlayerCollisionVolume)
  (params (OverlappedComponent :type Primitive%20Component%20Object%20Reference) (OtherActor :type Actor%20Object%20Reference) (OtherComp :type Primitive%20Component%20Object%20Reference) (OtherBodyIndex :type Integer) (bFromSweep :type Boolean) (SweepResult :type Hit%20Result%20Structure%20%28by%20ref%29))
  ;[0.01429 0.1146 0.1123] OnAnalyticsPlayerCollisionVolumeOverlapped: Let the analytics volume actor know about it.
  (Networking|SwitchHasAuthority
    (:Authority
      (if (Utilities|Time|IsTimerActivebyHandle (Variables|Transient|GetAnalyticsCollectionDurationTimerHandle))
        (bind _analyticsvolumeactor (Variables|Transient|GetAnalyticsVolumeActor)
          (:then
            (Class|BWeaponBunkerAnalyticsVolumeActor|NotifyActorBunkerEntered _analyticsvolumeactor OtherActor))
          (:else))))))

(fn CreateMapIcon
  ;[0.02925 0.1838 0.1875] CreateMapIcon: Create the map icon actor if it doesn't exist.
  (Networking|SwitchHasAuthority
    (:Authority
      (Variables|Transient|GetWeaponBunkerMapIcon
        (:then
          _)
        (:else
          (bind _returnvalue (Game|SpawnActorfromClass "/MidMatchWeaponBunkerGameplay/Items/WeaponBunkerDoor/B_WeaponBunkerMapIconActor.B_WeaponBunkerMapIconActor_C" (Transformation|GetActorTransform) "AlwaysSpawn" "MultiplyWithRoot" 0 0))
          (Variables|Transient|SetWeaponBunkerMapIcon _returnvalue)))
      (return (Variables|Transient|GetWeaponBunkerMapIcon)))))

(fn DestroyMapIcon
  ;[0.0285 0.1615 0.1566] DestroyMapIcon: Destroy the spawned map icon actor.
  (Networking|SwitchHasAuthority
    (:Authority
      (bind _weaponbunkermapicon (Variables|Transient|GetWeaponBunkerMapIcon)
        (:then
          (Actor|DestroyActor _weaponbunkermapicon)
          (Variables|Transient|SetWeaponBunkerMapIcon 0))
        (:else)))))

(fn GetMarkerDetailsId
  (return (select (Utilities|Enum|Equal(Enum) (Variables|Transient|GetWeaponBunkerDoorState) (LocalVariables|GetOpenedState)) "MarkerDetails.MMObj.WeaponBunkerGameplay.Open" "MarkerDetails.MMObj.WeaponBunkerGameplay")))

(fn GetPrimingDuration
  (return (Math|Float|Max(Float) (ScalableFloat|GetValueatLevel (Variables|Hotfix|GetHotfix_PrimingDuration)) 1.0)))

(fn HandleAnalyticsCollectionDurationComplete
  ;[0.02595 0.1667 0.1617] HandleAnalyticsCollectionDurationComplete: Stop collecting analytics.
  (Networking|SwitchHasAuthority
    (:Authority
      (Utilities|Time|ClearandInvalidateTimerbyHandle (Variables|Transient|GetAnalyticsCollectionDurationTimerHandle))
      (bind _analyticsvolumeactor (Variables|Transient|GetAnalyticsVolumeActor)
        (:then
          (Class|BWeaponBunkerAnalyticsVolumeActor|SetEnableAnalyticsVolume _analyticsvolumeactor false)
          (Collision|SetCollisionEnabled (Variables|Default|GetAnalyticsPlayerCollisionVolume) "NoCollision"))
        (:else)))))

(fn RefreshMarkerDisplayInfo
  ;[0.0151 0.1615 0.1582] RefreshMarkerDisplayInfo: Update the marker display info based on the marker details id.
  (bind _weaponbunkermarkedactordisplayinfodatatable (Variables|General|GetWeaponBunkerMarkedActorDisplayInfoDatatable)
    (:then
      (bind _returnvalue (Utilities|GetDataTableRow _weaponbunkermarkedactordisplayinfodatatable (MarkableActor|GetMarkerDetailsId)))
      (Variables|Marking|SetMarkerDisplay _returnvalue))
    (:else)))

(fn SetMapIconOpened
  ;[0.1161 0.1562 0.1419] SetMapIconOpened: Create the map icon actor, if it doesn't exist, and set the map icon.
  (Networking|SwitchHasAuthority
    (:Authority
      (bind _mapiconactor (CallFunction|CreateMapIcon))
      (Utilities|IsValid _mapiconactor
        (:Is%20Valid
          (Class|BWeaponBunkerMapIconActor|ActivateOpened _mapiconactor))))))

(fn SetMapIconOpening
  ;[0.1161 0.1562 0.1419] SetMapIconOpening: Create the map icon actor, if it doesn't exist, and set the map icon.
  (Networking|SwitchHasAuthority
    (:Authority
      (bind _mapiconactor (CallFunction|CreateMapIcon))
      (Utilities|IsValid _mapiconactor
        (:Is%20Valid
          (Class|BWeaponBunkerMapIconActor|ActivateOpening _mapiconactor))))))

(fn SetWeaponBunkerDoorState
  (params (NewState :type WeaponBunkerDoorOpeningStateEnum%20Enum :default "NewEnumerator0"))
  (bind _weaponbunkerdoorstate (Variables|Transient|GetWeaponBunkerDoorState))
  ;[0.02494 0.105 0.1094] SetWeaponBunkerDoorState: Update the state of the weapon bunker door.
  (Networking|SwitchHasAuthority
    (:Authority
      (if (Utilities|Enum|NotEqual(Enum) NewState _weaponbunkerdoorstate)
        (Networking|FlushNetDormancy)
        (Fortnite|FortShippingLog self (Utilities|String|Append "SetWeaponBunkerDoorState: Switching state from " (Utilities|String|EnumtoString _weaponbunkerdoorstate) " to " (Utilities|String|EnumtoString NewState)))
        (bind _output_get (Variables|Transient|SetWeaponBunkerDoorState NewState))
        (CallFunction|ClientWeaponBunkerDoorStateChanged _output_get) ; Useful for handling burst events.
      ))))

(fn ConstructionScript
  (|CallParentFunction)
  (if (Editor|InEditor)
    (Transformation|SetRelativeTransform (Variables|Default|GetStaticMeshDoor) (Math|Transform|MakeTransform (Math|Vector|MakeVector 0.0 (Variables|Setting|GetDoorRelativeLocationY) (Variables|Setting|GetDoorOpenRelativeLocationZStart)) (Math|Rotator|MakeRotator (Variables|Setting|GetDoorOpenRollAngleStart))) false false)))
