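; Map icon actor for the mid-match weapon bunker door, switching between an opening icon and an opened icon.
; It holds two AthenaSpecialActor components, and ActivateOpening and ActivateOpened each activate one and deactivate the other, on authority only.
; Both functions flush net dormancy before toggling, and BeginPlay also flushes it, so the dormant actor replicates to clients.
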(event EventBeginPlay
  ;[0.007568 0.06955 0.07812] BeginPlay: Replicate to the client.
  (|CallParentFunction)
  (Networking|FlushNetDormancy))

(fn ActivateOpened
  ;[0.05948 0.1927 0.1793] ActivateOpened: Set to the opened map icon.
  (Networking|SwitchHasAuthority
    (:Authority
      (Networking|FlushNetDormancy)
      (Components|Activation|Activate (Variables|Default|GetAthenaSpecialActor_Opened) false)
      (Components|Activation|Deactivate (Variables|Default|GetAthenaSpecialActor_Opening)))))

(fn ActivateOpening
  ;[0.1492 0.04639 0.1823] ActivateOpened: Set to the opening map icon.
  (Networking|SwitchHasAuthority
    (:Authority
      (Networking|FlushNetDormancy)
      (Components|Activation|Activate (Variables|Default|GetAthenaSpecialActor_Opening) false)
      (Components|Activation|Deactivate (Variables|Default|GetAthenaSpecialActor_Opened)))))

(fn ConstructionScript)
