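# Light function material for the weapon bunker door klaxon, producing a moving warning-light pattern compatible with the light function atlas.
# Horizontally doubled UVs pan at the Speed parameter rate and wrap through Frac, feeding a soft centered SphereMask that a Power node tightens into two scrolling spots.
# A second SphereMask on the V coordinate multiplies the result, fading it toward top and bottom without using any textures.
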Material /MidMatchWeaponBunkerGameplay/Items/WeaponBunkerDoor/M_LF_WeaponBunkerKlaxon
graph {
 properties {
  bForceCompatibleWithLightFunctionAtlas=true
  materialDomain=MD_LightFunction
  pixelDepthOffsetMode=PDOM_Legacy
  usedShadingModels=Default Lit
 }
 nodes [
  n1 = ScalarParameter() {defaultValue=0.8, parameterName=Speed}
  n2 = Constant()
  n3 = Frac(None=Panner(Coordinate=TextureCoordinate() {uTiling=2}, Speed=AppendVector(A=@n1, B=@n2)) {bFractionalPart=true, speedX=1})
  n4 = Constant() {r=0.5}
  n5 = Constant()
  n6 = TextureCoordinate()
  Multiply(
   A=Power(Base=SphereMask(A=@n3, B=@n4, Radius=@n4, Hardness=@n5)) {constExponent=3}
   B=SphereMask(A=ComponentMask(None=@n6) {g=true}, B=Constant() {r=0.5}, Radius=Constant() {r=0.45}, Hardness=Constant() {r=0.2})
  )
 ]
}
