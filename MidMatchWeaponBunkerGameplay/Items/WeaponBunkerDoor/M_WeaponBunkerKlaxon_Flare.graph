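# Additive unlit lens-flare material for the weapon bunker door klaxon, drawn on a camera-aligned mesh pushed toward the camera by CameraBias.
# Two counter-scrolling lens texture samples multiply, fade toward the horizontal UV edges and add to a mask texture.
# A static switch tints it with a single or time-of-day color, and a Speed-driven sine pulses it.
# Scene depth sampled at the biased pivot's screen position fades the flare when occluded.
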Material /MidMatchWeaponBunkerGameplay/Items/WeaponBunkerDoor/M_WeaponBunkerKlaxon_Flare
graph {
 properties {blendMode=BLEND_Additive, pixelDepthOffsetMode=PDOM_Legacy, shadingModel=MSM_Unlit, usedShadingModels=Unlit}
 nodes [
  n1 = CameraPositionWS()
  n2 = ObjectPositionWS()
  n3 = Multiply(A=Normalize(VectorInput=Subtract(A=@n1, B=@n2)), B=ScalarParameter() {defaultValue=25, parameterName=CameraBias})
  Add(
   A=@n3
   B=MaterialFunctionCall!(
    WorldPosition=WorldPosition().XYZ
    PivotLocation
    WS Normals
    Custom Object Basis 1=Constant3Vector() {constant=<rgba 0 1 0 1>}
    Custom Object Basis 2=Constant3Vector() {constant=<rgba 0 0 -1 0>}
    Custom Object Basis 3
   ) {materialFunction=/Engine/Functions/Engine_MaterialFunctions02/WorldPositionOffset/AlignMeshToTheCamera}.World Position Offset
  )
  n4 = TextureCoordinate()
  n5 = ComponentMask(None=@n4) {r=true}
  n6 = Multiply(A=Time()) {constB=0.1}
  n7 = Subtract(A=Multiply(A=@n5) {constB=3}, B=@n6)
  n8 = ComponentMask(None=@n4) {g=true}
  n9 = Add(A=Multiply(A=@n5) {constB=1.5}, B=@n6)
  n10 = Multiply(
   A=TextureSample(UVs=AppendVector(A=@n7, B=@n8)) {samplerType=SAMPLERTYPE_LinearColor, texture=&/T_WeaponBunkerKlaxon_Flare_Lens}.R
   B=TextureSample(UVs=AppendVector(A=@n9, B=@n8)) {samplerType=SAMPLERTYPE_LinearColor, texture=&T_WeaponBunkerKlaxon_Flare_Lens}.R
  )
  n11 = Add(A=@n5) {constB=-0.5}
  n12 = OneMinus(None=Multiply(A=Abs(None=@n11)) {constB=2})
  n13 = Multiply(A=@n12, B=@n12)
  n14 = TextureSample() {samplerType=SAMPLERTYPE_LinearGrayscale, texture=&/T_WeaponBunkerKlaxon_Flare_Mask}
  n15 = Multiply(A=Add(A=Multiply(A=@n10, B=@n13), B=@n14.R)) {constB=1.6}
  n16 = StaticSwitchParameter(
   True=MaterialFunctionCall(
    Color_Morning=VectorParameter() {parameterName=FlareColorMorning, defaultValue=<rgba 1 0 0 0>}.RGB
    Color_Day=VectorParameter() {parameterName=FlareColorDay, defaultValue=<rgba 0 1 0 0>}.RGB
    Color_Evening=VectorParameter() {parameterName=FlareColorEvening, defaultValue=<rgba 0 0 1 0>}.RGB
    Color_Night=VectorParameter() {parameterName=FlareColorNight, defaultValue=<rgba 1 1 1 1>}.RGB
   ) {materialFunction=&Color_Time_of_Day}.Result
   False=VectorParameter() {parameterName=FlareColor, defaultValue=<rgba 0.4 0.0876 0.0253 1>}.RGB
  ) {parameterName=UseTODFlareColors}
  n17 = ObjectPositionWS()
  n18 = WorldPosition()
  n19 = Add(A=Add(A=WorldPosition().XYZ, B=Subtract(A=@n17, B=@n18.XYZ)), B=@n3)
  n20 = Subtract(
   A=SceneDepth(UVs=MaterialFunctionCall(World Position=@n19) {materialFunction=&TransformToClipSpace}.Screen aligned 0-1 UV).R
   B=Subtract(
    A=PixelDepth().R
    B=MaterialFunctionCall(Vector 3=@n3) {materialFunction=/Engine/Functions/Engine_MaterialFunctions02/Utility/VectorLength}.V3 Length
   )
  )
  n21 = Multiply(A=Time(), B=ScalarParameter() {defaultValue=1, parameterName=Speed})
  Multiply(A=Multiply(A=Multiply(A=@n15, B=@n16), B=Saturate(None=@n20)), B=ConstantBiasScale(None=Sine(None=@n21)))
 ]
}
