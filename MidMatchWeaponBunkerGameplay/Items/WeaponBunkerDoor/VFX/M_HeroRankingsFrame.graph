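# Unlit additive Niagara sprite material for the hero rankings frame VFX on the weapon bunker door.
# The red channel of the bunker lock ranking frame texture masks the particle color, which passes through the shared MF_FX_EmissiveAdjustments function for emissive.
# The same mask also multiplies particle alpha, so particles control both tint and fade.
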Material /MidMatchWeaponBunkerGameplay/Items/WeaponBunkerDoor/VFX/M_HeroRankingsFrame
graph {
 properties {
  bAutomaticallySetUsageInEditor=false
  bUsedWithNiagaraSprites=true
  blendMode=BLEND_Additive
  shadingModel=MSM_Unlit
  usedShadingModels=Unlit
 }
 nodes [
  n1 = ParticleColor()
  n2 = TextureSample() {texture=&/T_BunkerLock_Ranking_Frame}
  MaterialFunctionCall(Emissive=Multiply(A=@n1.RGB, B=@n2.R)) {materialFunction=&MF_FX_EmissiveAdjustments}
  Multiply(A=@n1.A, B=Multiply(A=@n2.R))
 ]
}
