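# Niagara system for the dynamic holographic effect on the weapon bunker door in mid-match weapon bunker gameplay.
# It holds three emitters: separate holo beam emitters named for locked and unlocked states, plus a Squares_LW emitter.
# User float parameters GameTime and InteractDuration, defaulting to a four-second interaction, are exposed to feed timing from outside.
# Only the system state module is listed, with no renderer details.
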NiagaraSystem /MidMatchWeaponBunkerGameplay/Items/WeaponBunkerDoor/VFX/NS_BunkerDoor_HoloEffect_Dynamic
dependencies {usedDataInterfaces [], usedDynamicInputs [], usedModules [&SystemState]}
emitters {
 HoloBeams_Locked {rendererClasses [], renderers []}
 HoloBeams_Unlocked {rendererClasses [], renderers []}
 Squares_LW {rendererClasses [], renderers []}
}
summary {
 systemName=NS_BunkerDoor_HoloEffect_Dynamic
 emitters [
  {emitterName=Squares_LW, rendererClasses []}
  {emitterName=HoloBeams_Locked, rendererClasses []}
  {emitterName=HoloBeams_Unlocked, rendererClasses []}
 ]
 userVariables [
  User.GameTime {type=NiagaraFloat, defaultValue=0, description=""}
  User.InteractDuration {type=NiagaraFloat, defaultValue=4, description=""}
 ]
}
