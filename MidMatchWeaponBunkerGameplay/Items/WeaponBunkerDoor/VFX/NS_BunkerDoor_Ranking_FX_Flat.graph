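# Niagara system for the weapon bunker door's flat ranking display effect, driven by a float Rank and a bool Unlocked user parameter.
# Five emitters, named for the ranking element, two sets of holo bits and two backdrops, make up the effect.
# System scripts use a Timeline module and variable assignments, with dynamic inputs for color and vector lerps, float switching and distance-based LOD remapping.
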NiagaraSystem /MidMatchWeaponBunkerGameplay/Items/WeaponBunkerDoor/VFX/NS_BunkerDoor_Ranking_FX_Flat
dependencies {
 usedDataInterfaces []
 usedDynamicInputs [
  &LerpLinearColors
  &Lerp_Vector2
  &Vector2DFromFloat
  &FN_LOD_Distance_Remap
  &Multiply_VectorByFloat
  &MakeVector
  &MergeHelper_FloatSwitch
  &MakeIntFromFloat
  &MakeFloatFromVector2d
 ]
 usedModules [
  &SystemState
  &Timeline
  &self:SystemSpawnScript.SystemScriptSource.SystemScriptGraph.NiagaraNodeAssignment_0.SetVariables_#1
  &self:SystemSpawnScript.SystemScriptSource.SystemScriptGraph.NiagaraNodeAssignment_1.SetVariables_#2
 ]
}
emitters {
 Em_HoloBits {rendererClasses [], renderers []}
 Em_HoloBits001 {rendererClasses [], renderers []}
 Em_Ranking {rendererClasses [], renderers []}
 Em_Ranking_Backdrop {rendererClasses [], renderers []}
 Em_Ranking_Backdrop001 {rendererClasses [], renderers []}
}
summary {
 systemName=NS_BunkerDoor_Ranking_FX_Flat
 emitters [
  {emitterName=Em_Ranking, rendererClasses []}
  {emitterName=Em_HoloBits, rendererClasses []}
  {emitterName=Em_HoloBits001, rendererClasses []}
  {emitterName=Em_Ranking_Backdrop, rendererClasses []}
  {emitterName=Em_Ranking_Backdrop001, rendererClasses []}
 ]
 userVariables [
  User.Rank {type=NiagaraFloat, defaultValue=0, description=""}
  User.Unlocked {type=NiagaraBool, defaultValue=true}
 ]
}
